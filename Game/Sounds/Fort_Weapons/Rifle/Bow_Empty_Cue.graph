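# Sound cue named for a bow's empty-weapon sound in the rifle weapon sounds, with no nodes and a null root, so it has no audio to play.
# It still carries the weapon foley sound class, a reduced volume and an attenuation falloff override.
# Its concurrency override allows a single instance and prevents new ones from starting while it plays.
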SoundCue /Game/Sounds/Fort_Weapons/Rifle/Bow_Empty_Cue
root=null
nodes []
properties {
 bOverrideConcurrency=true
 maxDistance=2097152
 soundClassObject=&Fort_WeaponFoley
 volumeMultiplier=0.251
 attenuationOverrides {falloffDistance=1600}
 concurrencyOverrides {maxCount=1, resolutionRule=PreventNew}
}
