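# Sound cue for the rifle's empty sound, judging by its name, in the weapon foley sound class.
# A single wave player feeds the output directly, reusing an auto rifle reload element sample at reduced volume with no randomization.
# The cue overrides concurrency to allow only a couple of simultaneous instances, stopping the oldest when the limit is exceeded.
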SoundCue /Game/Sounds/Fort_Weapons/Rifle/Rifle_Empty_Cue
root=WavePlayer_21
nodes [{id=WavePlayer_21, properties {soundWaveAssetPtr=/Game/Sounds/Fort_Weapons/AutoRifle/Reload_Elements/assault_end_sdk_01}}]
properties {
 bOverrideConcurrency=true
 maxDistance=2097152
 soundClassObject=&Fort_WeaponFoley
 volumeMultiplier=0.251
 attenuationOverrides {falloffDistance=1600}
 concurrencyOverrides {maxCount=2, resolutionRule=StopOldest}
}
