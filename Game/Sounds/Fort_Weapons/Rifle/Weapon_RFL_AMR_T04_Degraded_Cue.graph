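# Sound cue for third-person fire of the degraded T04 AMR rifle, using electric shot samples.
# A quality node selects between two near-identical mixes, the second with one distant variant instead of two, and leaves its other two slots empty.
# Each mix layers randomized close shots with a distance crossfade from medium to distant sounds.
# Layers carry their own attenuation and sound class, and close and distant samples pass through modulators.
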SoundCue /Game/Sounds/Fort_Weapons/Rifle/Weapon_RFL_AMR_T04_Degraded_Cue
root=QualityLevel_0
nodes [
 {id=Attenuation_3, children [Modulator_3], properties {attenuationSettings=/Game/Sounds/Attenuation/Fort_Weapon_Rifle_Sml_Att}}
 {id=Attenuation_4, children [Modulator_5], properties {attenuationSettings=/Game/Sounds/Attenuation/Fort_Weapon_Projectile_Medium_StW_Att}}
 {id=Attenuation_5, children [Modulator_4], properties {attenuationSettings=/Game/Sounds/Attenuation/Fort_Weapon_Projectile_Distant_StW_Att}}
 {id=Attenuation_6, children [Modulator_6], properties {attenuationSettings=&Fort_Weapon_Rifle_Sml_Att}}
 {id=Attenuation_7, children [Modulator_8], properties {attenuationSettings=&Fort_Weapon_Projectile_Medium_StW_Att}}
 {id=Attenuation_8, children [Modulator_7], properties {attenuationSettings=&Fort_Weapon_Projectile_Distant_StW_Att}}
 {
  id=DistanceCrossFade_1
  children [Attenuation_4, Attenuation_5]
  properties {crossFadeInput [
    {fadeInDistanceEnd=100, fadeInDistanceStart=0, fadeOutDistanceEnd=3000, fadeOutDistanceStart=2000, volume=1}
    {fadeInDistanceEnd=3000, fadeInDistanceStart=1000, fadeOutDistanceEnd=12000, fadeOutDistanceStart=10000, volume=1}
   ]}
 }
 {
  id=DistanceCrossFade_2
  children [Attenuation_7, Attenuation_8]
  properties {crossFadeInput [
    {fadeInDistanceEnd=100, fadeInDistanceStart=0, fadeOutDistanceEnd=3000, fadeOutDistanceStart=2000, volume=1}
    {fadeInDistanceEnd=3000, fadeInDistanceStart=1000, fadeOutDistanceEnd=12000, fadeOutDistanceStart=10000, volume=1}
   ]}
 }
 {id=Mixer_1, children [SoundClass_2,SoundClass_3], properties {inputVolume [1,1]}}
 {id=Mixer_2, children [SoundClass_4,SoundClass_5], properties {inputVolume [1,1]}}
 {id=Modulator_3, children [Random_3]}
 {id=Modulator_4, children [Random_2]}
 {id=Modulator_5, children [WavePlayer_7]}
 {id=Modulator_6, children [Random_5]}
 {id=Modulator_7, children [Random_4]}
 {id=Modulator_8, children [WavePlayer_12]}
 {id=QualityLevel_0, children [Mixer_1,Mixer_2,null,null]}
 {id=Random_2, children [WavePlayer_6,WavePlayer_5], properties {weights [1,1]}}
 {id=Random_3, children [WavePlayer_8,WavePlayer_9], properties {weights [1,1]}}
 {id=Random_4, children [WavePlayer_10], properties {weights [1]}}
 {id=Random_5, children [WavePlayer_13, WavePlayer_14], properties {weights [1,1]}}
 {
  id=SoundClass_2
  children [Attenuation_3]
  properties {soundClassOverride=&Fort_WeaponRanged_Fire_3P_PSM}
 }
 {
  id=SoundClass_3
  children [DistanceCrossFade_1]
  properties {soundClassOverride=&Fort_WeaponRanged_Fire_Distant}
 }
 {
  id=SoundClass_4
  children [Attenuation_6]
  properties {soundClassOverride=&Fort_WeaponRanged_Fire_3P_PSM}
 }
 {
  id=SoundClass_5
  children [DistanceCrossFade_2]
  properties {soundClassOverride=&Fort_WeaponRanged_Fire_Distant}
 }
 {id=WavePlayer_10, properties {soundWaveAssetPtr=/Game/Sounds/Fort_Weapons/Rifle/AMR/Mono/Rifle_AMR_Shot_Distant06e}}
 {id=WavePlayer_12, properties {soundWaveAssetPtr=&/rifle_medium_02}}
 {id=WavePlayer_13, properties {soundWaveAssetPtr=/Game/Sounds/Fort_Weapons/Rifle/AMR_Electric/rifle_amr_electric_shot_03}}
 {id=WavePlayer_14, properties {soundWaveAssetPtr=/Game/Sounds/Fort_Weapons/Rifle/AMR_Electric/rifle_amr_electric_shot_04}}
 {id=WavePlayer_5, properties {soundWaveAssetPtr=&Rifle_AMR_Shot_Distant06e}}
 {id=WavePlayer_6, properties {soundWaveAssetPtr=/Game/Sounds/Fort_Weapons/Rifle/AMR/Mono/Rifle_AMR_Shot_Distant06a}}
 {id=WavePlayer_7, properties {soundWaveAssetPtr=&rifle_medium_02}}
 {id=WavePlayer_8, properties {soundWaveAssetPtr=&rifle_amr_electric_shot_03}}
 {id=WavePlayer_9, properties {soundWaveAssetPtr=&rifle_amr_electric_shot_04}}
]
properties {
 bOverrideConcurrency=true
 maxDistance=20000
 volumeMultiplier=0.7
 attenuationOverrides {distanceAlgorithm=NaturalSound, falloffDistance=4000, attenuationShapeExtents {x=1000}}
 concurrencyOverrides {resolutionRule=PreventNew}
}
