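# Sound cue named as the destroy sound for the T04 AMR rifle, built from gunshot waves in two mixed layers.
# A near layer randomly picks one of two electric AMR shots under small-rifle attenuation and the third-person ranged-fire sound class.
# A distant-fire class layer distance-crossfades a medium rifle shot into a distant AMR shot, each with its own attenuation.
# Every branch passes through a modulator, and the cue overrides concurrency to prevent new instances at the limit.
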SoundCue /Game/Sounds/Fort_Weapons/Rifle/Weapon_RFL_AMR_T04_Destroy_Cue
root=Mixer_4
nodes [
 {id=Attenuation_35, children [Modulator_5], properties {attenuationSettings=/Game/Sounds/Attenuation/Fort_Weapon_Rifle_Sml_Att}}
 {id=Attenuation_36, children [Modulator_15], properties {attenuationSettings=/Game/Sounds/Attenuation/Fort_Weapon_Projectile_Medium_StW_Att}}
 {id=Attenuation_37, children [Modulator_14], properties {attenuationSettings=/Game/Sounds/Attenuation/Fort_Weapon_Projectile_Distant_StW_Att}}
 {
  id=DistanceCrossFade_10
  children [Attenuation_36, Attenuation_37]
  properties {crossFadeInput [
    {fadeInDistanceEnd=100, fadeInDistanceStart=0, fadeOutDistanceEnd=3000, fadeOutDistanceStart=2000, volume=1}
    {fadeInDistanceEnd=3000, fadeInDistanceStart=1000, fadeOutDistanceEnd=12000, fadeOutDistanceStart=10000, volume=1}
   ]}
 }
 {id=Mixer_4, children [SoundClass_25, SoundClass_26], properties {inputVolume [1,1]}}
 {id=Modulator_14, children [Random_18]}
 {id=Modulator_15, children [WavePlayer_37]}
 {id=Modulator_5, children [Random_10]}
 {id=Random_10, children [WavePlayer_21, WavePlayer_22], properties {weights [1,1]}}
 {id=Random_18, children [WavePlayer_35], properties {weights [1]}}
 {
  id=SoundClass_25
  children [Attenuation_35]
  properties {soundClassOverride=&Fort_WeaponRanged_Fire_3P_PSM}
 }
 {
  id=SoundClass_26
  children [DistanceCrossFade_10]
  properties {soundClassOverride=&Fort_WeaponRanged_Fire_Distant}
 }
 {id=WavePlayer_21, properties {soundWaveAssetPtr=/Game/Sounds/Fort_Weapons/Rifle/AMR_Electric/rifle_amr_electric_shot_03}}
 {id=WavePlayer_22, properties {soundWaveAssetPtr=/Game/Sounds/Fort_Weapons/Rifle/AMR_Electric/rifle_amr_electric_shot_04}}
 {id=WavePlayer_35, properties {soundWaveAssetPtr=/Game/Sounds/Fort_Weapons/Rifle/AMR/Mono/Rifle_AMR_Shot_Distant06e}}
 {id=WavePlayer_37, properties {soundWaveAssetPtr=&/rifle_medium_02}}
]
properties {
 bOverrideConcurrency=true
 maxDistance=20000
 volumeMultiplier=0.7
 attenuationOverrides {distanceAlgorithm=NaturalSound, falloffDistance=4000, attenuationShapeExtents {x=1000}}
 concurrencyOverrides {resolutionRule=PreventNew}
}
