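# Sound cue for the charge sound of the Epic tier-four AMR rifle.
# A single wave player plays amr_charge_02 with no randomization or layering, routed to the third-person ranged weapon fire sound class.
# Spatial falloff references the shared small-rifle attenuation asset.
# A concurrency override limits it to one instance per owner and blocks new plays while one is active, so charge sounds do not stack.
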SoundCue /Game/Sounds/Fort_Weapons/Rifle/Weapon_RFL_AMR_T04_Epic_AMR_Charge_Cue
root=WavePlayer_2
nodes [{id=WavePlayer_2, properties {soundWaveAssetPtr=&/amr_charge_02}}]
properties {
 attenuationSettings=/Game/Sounds/Attenuation/Fort_Weapon_Rifle_Sml_Att
 bOverrideConcurrency=true
 maxDistance=3000
 soundClassObject=&Fort_WeaponRanged_Fire_3P_PSM
 volumeMultiplier=1
 attenuationOverrides {distanceAlgorithm=NaturalSound, falloffDistance=4000, attenuationShapeExtents {x=1000}}
 concurrencyOverrides {bLimitToOwner=true, bVolumeScaleCanRelease=true, maxCount=1, resolutionRule=PreventNew, volumeScaleReleaseTime=1}
}
