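# Sound cue for the full-charge loop of the Epic tier 4 AMR rifle.
# A looping charged wave feeds an Enveloper that ramps volume from silence to full, then a small-rifle Attenuation node and a root SoundClass node overriding to the third-person ranged-fire class.
# The cue itself sets a low volume multiplier, iron-sights attenuation and a concurrency override that prevents new instances once the limit is reached.
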SoundCue /Game/Sounds/Fort_Weapons/Rifle/Weapon_RFL_AMR_T04_Epic_AMR_FullCharge_Cue
root=SoundClass_25
nodes [
 {id=Attenuation_35, children [Enveloper_0], properties {attenuationSettings=/Game/Sounds/Attenuation/Fort_Weapon_Rifle_Sml_Att}}
 {
  id=Enveloper_0
  children [WavePlayer_17]
  properties {pitchCurve {editorCurveData {keys [{time=0}]}}, volumeCurve {editorCurveData {keys [<at 0 0>,<at 1 1>]}}}
 }
 {
  id=SoundClass_25
  children [Attenuation_35]
  properties {soundClassOverride=&Fort_WeaponRanged_Fire_3P_PSM}
 }
 {id=WavePlayer_17, properties {bLooping=true, soundWaveAssetPtr=&/amr_charged_loop_01}}
]
properties {
 attenuationSettings=/Game/Sounds/Attenuation/Fort_Weapon_IronSights_Att
 bOverrideConcurrency=true
 maxDistance=3000
 soundClassObject=/Game/Sounds/FortSoundClassesAndModes/Fort_Weapons
 volumeMultiplier=0.1481
 attenuationOverrides {distanceAlgorithm=NaturalSound, falloffDistance=4000, attenuationShapeExtents {x=1000}}
 concurrencyOverrides {resolutionRule=PreventNew}
}
