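# Sound cue for a looping tone on the epic tier-4 AMR rifle, in the weapons sound class.
# A quality-level switch plays amr_tone_02 at the first level and the "opt" wave variant at the other three.
# Both looping branches pass through small-rifle attenuation nodes, while the cue itself references iron-sights attenuation.
# A concurrency override blocks new instances once its limit is reached.
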SoundCue /Game/Sounds/Fort_Weapons/Rifle/Weapon_RFL_AMR_T04_Epic_AMR_Tone_2_Cue
root=QualityLevel_0
nodes [
 {id=Attenuation_0, children [WavePlayer_0], properties {attenuationSettings=/Game/Sounds/Attenuation/Fort_Weapon_Rifle_Sml_Att}}
 {id=Attenuation_35, children [WavePlayer_5], properties {attenuationSettings=&Fort_Weapon_Rifle_Sml_Att}}
 {id=QualityLevel_0, children [Attenuation_35, Attenuation_0, Attenuation_0, Attenuation_0]}
 {id=WavePlayer_0, properties {bLooping=true, soundWaveAssetPtr=&/amr_tone_02_opt}}
 {id=WavePlayer_5, properties {bLooping=true, soundWaveAssetPtr=&/amr_tone_02}}
]
properties {
 attenuationSettings=/Game/Sounds/Attenuation/Fort_Weapon_IronSights_Att
 bOverrideConcurrency=true
 maxDistance=3000
 soundClassObject=/Game/Sounds/FortSoundClassesAndModes/Fort_Weapons
 volumeMultiplier=0.3704
 attenuationOverrides {distanceAlgorithm=NaturalSound, falloffDistance=4000, attenuationShapeExtents {x=1000}}
 concurrencyOverrides {resolutionRule=PreventNew}
}
