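# First-person fire sound cue for the Epic T04 anti-material rifle in Athena, using first-person gunfire sound class, attenuation and concurrency assets.
# A mixer layers a random shaped-charge blast, a .357 low-ammo sample and a charge-driven electric shot.
# The electric layer picks one of two shots at random, varies its pitch upward, delays it slightly, and sets its level through a crossfade on the ChargePercent parameter.
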SoundCue /Game/Sounds/Fort_Weapons/Rifle/Weapon_RFL_AMR_T04_Epic_Anti_Material_Rifle_Athena_1P_Cue
root=Mixer_3
nodes [
 {id=Delay_1, children [Modulator_3], properties {delayMax=0.1, delayMin=0.1}}
 {id=Mixer_3, children [Random_5, WavePlayer_15, ParamCrossFade_0], properties {inputVolume [1,1,1]}}
 {id=Modulator_3, children [Random_4], properties {pitchMax=1.2, pitchMin=1}}
 {
  id=ParamCrossFade_0
  children [Delay_1]
  properties {
   paramName=ChargePercent
   crossFadeInput [{fadeInDistanceEnd=1, fadeInDistanceStart=0, fadeOutDistanceEnd=100, fadeOutDistanceStart=1, volume=1}]
  }
 }
 {id=Random_4, children [WavePlayer_11, WavePlayer_12], properties {weights [1,1]}}
 {id=Random_5, children [WavePlayer_13, WavePlayer_14], properties {weights [1,1]}}
 {id=WavePlayer_11, properties {soundWaveAssetPtr=/Game/Sounds/Fort_Weapons/Rifle/AMR_Electric/rifle_amr_electric_shot_03}}
 {id=WavePlayer_12, properties {soundWaveAssetPtr=/Game/Sounds/Fort_Weapons/Rifle/AMR_Electric/rifle_amr_electric_shot_04}}
 {id=WavePlayer_13, properties {soundWaveAssetPtr=/Game/Sounds/Fort_Gadgets/Outlander/ShapedCharge_001}}
 {id=WavePlayer_14, properties {soundWaveAssetPtr=/Game/Sounds/Fort_Gadgets/Outlander/ShapedCharge_004}}
 {id=WavePlayer_15, properties {soundWaveAssetPtr=/Game/Sounds/Fort_Weapons/LowAmmo/357_low_ammo_01}}
]
properties {
 attenuationSettings=&Fort_Weapon_Projectile_FirstPerson_Athena_Att
 maxDistance=2097152
 soundClassObject=&Fort_WeaponRanged_Fire_1P_PSM
 volumeMultiplier=0.8
 attenuationOverrides {distanceAlgorithm=NaturalSound, falloffDistance=4000, attenuationShapeExtents {x=1000}}
 concurrencyOverrides {bLimitToOwner=true, maxCount=2, resolutionRule=StopOldest}
 concurrencySet [&Weapon_Gunfire_FirstPerson_Conc]
}
