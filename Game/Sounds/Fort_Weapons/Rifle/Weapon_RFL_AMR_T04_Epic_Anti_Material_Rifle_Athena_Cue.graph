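# Gunshot SoundCue for the Epic T04 anti-material rifle in Athena, splitting close and distant fire into separate sound classes.
# Close fire mixes random shaped-charge blasts, a low-ammo sample and delayed, pitch-modulated electric shots.
# The first quality level fades electric shots by ChargePercent and adds delayed thunder to separately attenuated medium and distant layers.
# The second drops both features and distance-crossfades its simpler medium and distant layers; remaining quality slots are empty.
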SoundCue /Game/Sounds/Fort_Weapons/Rifle/Weapon_RFL_AMR_T04_Epic_Anti_Material_Rifle_Athena_Cue
root=QualityLevel_0
nodes [
 {id=Attenuation_3, children [Mixer_3], properties {attenuationSettings=&Fort_Weapon_Projectile_Close_Athena_Att}}
 {
  id=Attenuation_4
  children [Mixer_6]
  properties {attenuationSettings=&Fort_Weapon_Projectile_Medium_Mono_Athena_Att}
 }
 {
  id=Attenuation_5
  children [Mixer_7]
  properties {attenuationSettings=&Fort_Weapon_Projectile_Distant_Mono_Athena_Att}
 }
 {id=Attenuation_6, children [Mixer_5], properties {attenuationSettings=/Game/Sounds/Attenuation/Fort_Weapon_Rifle_Sml_Att}}
 {id=Attenuation_7, children [Modulator_8], properties {attenuationSettings=/Game/Sounds/Attenuation/Fort_Weapon_Projectile_Medium_StW_Att}}
 {id=Attenuation_8, children [Modulator_7], properties {attenuationSettings=/Game/Sounds/Attenuation/Fort_Weapon_Projectile_Distant_StW_Att}}
 {id=Delay_1, children [Modulator_3], properties {delayMax=0.1, delayMin=0.1}}
 {id=Delay_2, children [Modulator_6], properties {delayMax=0.1, delayMin=0.1}}
 {id=Delay_3, children [WavePlayer_2], properties {delayMax=0.3, delayMin=0.3}}
 {id=Delay_4, children [WavePlayer_3], properties {delayMax=1, delayMin=1}}
 {
  id=DistanceCrossFade_2
  children [Attenuation_7, Attenuation_8]
  properties {crossFadeInput [
    {fadeInDistanceEnd=100, fadeInDistanceStart=0, fadeOutDistanceEnd=3000, fadeOutDistanceStart=2000, volume=1}
    {fadeInDistanceEnd=3000, fadeInDistanceStart=1000, fadeOutDistanceEnd=12000, fadeOutDistanceStart=10000, volume=1}
   ]}
 }
 {id=Mixer_0, children [Attenuation_4, Attenuation_5], properties {inputVolume [1,0.5]}}
 {id=Mixer_2, children [SoundClass_2,SoundClass_3], properties {inputVolume [1,1]}}
 {id=Mixer_3, children [Random_5, WavePlayer_15, ParamCrossFade_0], properties {inputVolume [1,1,1]}}
 {id=Mixer_4, children [SoundClass_4,SoundClass_5], properties {inputVolume [1,1]}}
 {id=Mixer_5, children [Random_8, Delay_2, WavePlayer_23], properties {inputVolume [1,1,1]}}
 {id=Mixer_6, children [Modulator_5,Delay_3], properties {inputVolume [1,1.5]}}
 {id=Mixer_7, children [Modulator_4,Delay_4], properties {inputVolume [1,1]}}
 {id=Modulator_3, children [Random_4], properties {pitchMax=1.2, pitchMin=1}}
 {id=Modulator_4, children [Random_3]}
 {id=Modulator_5, children [WavePlayer_10]}
 {id=Modulator_6, children [Random_7], properties {pitchMax=1.2, pitchMin=1}}
 {id=Modulator_7, children [Random_6]}
 {id=Modulator_8, children [WavePlayer_18]}
 {
  id=ParamCrossFade_0
  children [Delay_1]
  properties {
   paramName=ChargePercent
   crossFadeInput [{fadeInDistanceEnd=1, fadeInDistanceStart=0, fadeOutDistanceEnd=100, fadeOutDistanceStart=1, volume=1}]
  }
 }
 {id=QualityLevel_0, children [Mixer_2,Mixer_4,null,null]}
 {id=Random_3, children [WavePlayer_9,WavePlayer_8], properties {weights [1,1]}}
 {id=Random_4, children [WavePlayer_11, WavePlayer_12], properties {weights [1,1]}}
 {id=Random_5, children [WavePlayer_13, WavePlayer_14], properties {weights [1,1]}}
 {id=Random_6, children [WavePlayer_17], properties {weights [1]}}
 {id=Random_7, children [WavePlayer_19], properties {weights [1]}}
 {id=Random_8, children [WavePlayer_21, WavePlayer_22], properties {weights [1,1]}}
 {
  id=SoundClass_2
  children [Attenuation_3]
  properties {soundClassOverride=&Fort_WeaponRanged_Fire_3P_PSM}
 }
 {id=SoundClass_3, children [Mixer_0], properties {soundClassOverride=&Fort_WeaponRanged_Fire_Distant}}
 {
  id=SoundClass_4
  children [Attenuation_6]
  properties {soundClassOverride=&Fort_WeaponRanged_Fire_3P_PSM}
 }
 {
  id=SoundClass_5
  children [DistanceCrossFade_2]
  properties {soundClassOverride=&Fort_WeaponRanged_Fire_Distant}
 }
 {id=WavePlayer_10, properties {soundWaveAssetPtr=&/rifle_medium_02}}
 {id=WavePlayer_11, properties {soundWaveAssetPtr=/Game/Sounds/Fort_Weapons/Rifle/AMR_Electric/rifle_amr_electric_shot_03}}
 {id=WavePlayer_12, properties {soundWaveAssetPtr=/Game/Sounds/Fort_Weapons/Rifle/AMR_Electric/rifle_amr_electric_shot_04}}
 {id=WavePlayer_13, properties {soundWaveAssetPtr=/Game/Sounds/Fort_Gadgets/Outlander/ShapedCharge_001}}
 {id=WavePlayer_14, properties {soundWaveAssetPtr=/Game/Sounds/Fort_Gadgets/Outlander/ShapedCharge_004}}
 {id=WavePlayer_15, properties {soundWaveAssetPtr=/Game/Sounds/Fort_Weapons/LowAmmo/357_low_ammo_01}}
 {id=WavePlayer_17, properties {soundWaveAssetPtr=/Game/Sounds/Fort_Weapons/Rifle/AMR/Mono/Rifle_AMR_Shot_Distant06a}}
 {id=WavePlayer_18, properties {soundWaveAssetPtr=&rifle_medium_02}}
 {id=WavePlayer_19, properties {soundWaveAssetPtr=&rifle_amr_electric_shot_03}}
 {id=WavePlayer_2, properties {soundWaveAssetPtr=/Game/Athena/Sounds/Weapons/AMR/sfx_amr_fire_thunderNew_02}}
 {id=WavePlayer_21, properties {soundWaveAssetPtr=&ShapedCharge_001}}
 {id=WavePlayer_22, properties {soundWaveAssetPtr=&ShapedCharge_004}}
 {id=WavePlayer_23, properties {soundWaveAssetPtr=&357_low_ammo_01}}
 {id=WavePlayer_3, properties {soundWaveAssetPtr=/Game/Athena/Sounds/Weapons/AMR/sfx_amr_fire_thunderNew_03}}
 {id=WavePlayer_8, properties {soundWaveAssetPtr=/Game/Sounds/Fort_Weapons/Rifle/AMR/Mono/Rifle_AMR_Shot_Distant06e}}
 {id=WavePlayer_9, properties {soundWaveAssetPtr=&Rifle_AMR_Shot_Distant06a}}
]
properties {
 maxDistance=27500
 volumeMultiplier=0.8
 attenuationOverrides {distanceAlgorithm=NaturalSound, falloffDistance=4000, attenuationShapeExtents {x=1000}}
 concurrencyOverrides {bLimitToOwner=true, maxCount=2, resolutionRule=StopOldest}
 concurrencySet [&Weapon_Gunfire_Local_Conc, &Weapon_Gunfire_Global_Conc]
}
