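# Sound cue for the low-ammo warning of the T04 AMR rifle weapon.
# A single wave player plays the AMR beam low-ammo sample at reduced volume, routed through the weapon foley sound class.
# It uses small-rifle attenuation settings, alongside natural-sound attenuation override values.
# Its own concurrency override prevents new instances from starting once the limit is reached.
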SoundCue /Game/Sounds/Fort_Weapons/Rifle/Weapon_RFL_AMR_T04_LowAmmo_Cue
root=WavePlayer_1
nodes [{id=WavePlayer_1, properties {soundWaveAssetPtr=/Game/Sounds/Fort_Weapons/AMR/AMR_Beam_LowAmmo_01}}]
properties {
 attenuationSettings=/Game/Sounds/Attenuation/Fort_Weapon_Rifle_Sml_Att
 bOverrideConcurrency=true
 maxDistance=3000
 soundClassObject=&Fort_WeaponFoley
 volumeMultiplier=0.501
 attenuationOverrides {distanceAlgorithm=NaturalSound, falloffDistance=4000, attenuationShapeExtents {x=1000}}
 concurrencyOverrides {resolutionRule=PreventNew}
}
