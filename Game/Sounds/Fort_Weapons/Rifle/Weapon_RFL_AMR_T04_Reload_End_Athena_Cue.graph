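# Sound cue for the end of the AMR T04 rifle reload in Athena, using an electric raygun reload-end sound.
# A single wave player feeds the root directly, with no randomization or modulation, at reduced volume.
# It uses the shared weapon reload attenuation and a third-person ranged weapon fire sound class.
# A concurrency override blocks new instances once the limit is reached instead of stopping playing ones.
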SoundCue /Game/Sounds/Fort_Weapons/Rifle/Weapon_RFL_AMR_T04_Reload_End_Athena_Cue
root=WavePlayer_20
nodes [{id=WavePlayer_20, properties {soundWaveAssetPtr=/Game/Sounds/Fort_Weapons/Rifle/AMR_Electric/weapon_raygun_reload_end_athena}}]
properties {
 attenuationSettings=&Fort_Weapon_Reload_Att
 bOverrideConcurrency=true
 maxDistance=2250
 soundClassObject=&Fort_WeaponRanged_Fire_3P_PSM
 volumeMultiplier=0.6
 attenuationOverrides {distanceAlgorithm=NaturalSound, falloffDistance=4000, attenuationShapeExtents {x=1000}}
 concurrencyOverrides {resolutionRule=PreventNew}
}
