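# Sound cue for firing the Exterminator, a Fine-tier T03 automatic rifle, layering a main shot with distance-based layers.
# A mixer combines a stereo fire wave in the third-person fire class with a crossfade of two waves in the distant class.
# Each wave has its own modulator and attenuation, with rifle_medium_distant_02 on the nearer band and rifle_medium_02 on the farther.
# Concurrency is overridden to prevent new sounds at the limit.
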SoundCue /Game/Sounds/Fort_Weapons/Rifle/Weapon_RFL_Auto_T03_Fine_Exterminator_Cue
root=Mixer_6
nodes [
 {id=Attenuation_14, children [Modulator_18], properties {attenuationSettings=/Game/Sounds/Attenuation/Fort_Weapon_Projectile_Medium_StW_Att}}
 {id=Attenuation_15, children [Modulator_19], properties {attenuationSettings=/Game/Sounds/Attenuation/Fort_Weapon_Projectile_Distant_StW_Att}}
 {id=Attenuation_53, children [Modulator_2], properties {attenuationSettings=/Game/Sounds/Attenuation/Fort_Weapon_Rifle_Sml_Att}}
 {
  id=DistanceCrossFade_7
  children [Attenuation_14, Attenuation_15]
  properties {crossFadeInput [
    {fadeInDistanceEnd=100, fadeInDistanceStart=0, fadeOutDistanceEnd=3000, fadeOutDistanceStart=2000, volume=1}
    {fadeInDistanceEnd=3000, fadeInDistanceStart=1000, fadeOutDistanceEnd=12000, fadeOutDistanceStart=10000, volume=1}
   ]}
 }
 {id=Mixer_6, children [SoundClass_37, SoundClass_38], properties {inputVolume [1,1]}}
 {id=Modulator_18, children [WavePlayer_38]}
 {id=Modulator_19, children [WavePlayer_37]}
 {id=Modulator_2, children [WavePlayer_19]}
 {
  id=SoundClass_37
  children [Attenuation_53]
  properties {soundClassOverride=&Fort_WeaponRanged_Fire_3P_PSM}
 }
 {
  id=SoundClass_38
  children [DistanceCrossFade_7]
  properties {soundClassOverride=&Fort_WeaponRanged_Fire_Distant}
 }
 {id=WavePlayer_19, properties {soundWaveAssetPtr=&/Rifle_Fire_Stereo_01}}
 {id=WavePlayer_37, properties {soundWaveAssetPtr=&/rifle_medium_02}}
 {id=WavePlayer_38, properties {soundWaveAssetPtr=&/rifle_medium_distant_02}}
]
properties {
 bOverrideConcurrency=true
 maxDistance=20000
 volumeMultiplier=0.55
 attenuationOverrides {distanceAlgorithm=NaturalSound, falloffDistance=4000, attenuationShapeExtents {x=1000}}
 concurrencyOverrides {resolutionRule=PreventNew}
}
