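# Sniper rifle fire sound cue for Athena, defaulting to the third-person ranged weapon fire sound class.
# A mixer layers close, medium and distant shots, with the latter two at reduced input levels but boosted by volume modulators.
# Each layer picks a random wave, then passes through modulation, its own distance attenuation and a sound class override, with medium and distant sharing the Distant class.
# Local and global gunfire concurrency sets limit overlapping shots.
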SoundCue /Game/Sounds/Fort_Weapons/Rifle/Weapon_RFL_Sniper_Athena_Cue
root=Mixer_1
nodes [
 {
  id=Attenuation_11
  children [Modulator_17]
  properties {attenuationSettings=&Fort_Weapon_Projectile_Close_Athena_Att}
 }
 {
  id=Attenuation_12
  children [Modulator_18]
  properties {attenuationSettings=&Fort_Weapon_Projectile_Medium_Mono_Athena_Att}
 }
 {
  id=Attenuation_13
  children [Modulator_19]
  properties {attenuationSettings=&Fort_Weapon_Projectile_Distant_Mono_Athena_Att}
 }
 {id=Mixer_1, children [SoundClass_14, SoundClass_16, SoundClass_15], properties {inputVolume [1,0.35,0.2]}}
 {id=Modulator_17, children [Random_15], properties {pitchMin=1}}
 {id=Modulator_18, children [Random_0], properties {volumeMax=2.05, volumeMin=1.95}}
 {id=Modulator_19, children [Random_17], properties {volumeMax=2.05, volumeMin=1.95}}
 {id=Random_0, children [WavePlayer_3,WavePlayer_4], properties {bShouldExcludeFromBranchCulling=true, weights [1,1]}}
 {id=Random_15, children [WavePlayer_33, WavePlayer_34, WavePlayer_35], properties {bShouldExcludeFromBranchCulling=true, weights [1,1,1]}}
 {id=Random_17, children [WavePlayer_38, WavePlayer_39], properties {bShouldExcludeFromBranchCulling=true, weights [1,1]}}
 {
  id=SoundClass_14
  children [Attenuation_11]
  properties {soundClassOverride=&Fort_WeaponRanged_Fire_3P_PSM}
 }
 {
  id=SoundClass_15
  children [Attenuation_13]
  properties {soundClassOverride=&Fort_WeaponRanged_Fire_Distant}
 }
 {
  id=SoundClass_16
  children [Attenuation_12]
  properties {soundClassOverride=&Fort_WeaponRanged_Fire_Distant}
 }
 {id=WavePlayer_3, properties {soundWaveAssetPtr=/Game/Sounds/Fort_Weapons/SniperRifles/Sniper_Rifle_Fire_MediumNew_New_01}}
 {id=WavePlayer_33, properties {soundWaveAssetPtr=/Game/Sounds/Fort_Weapons/SniperRifles/Sniper_Rifle_Fire_Close_New_01}}
 {id=WavePlayer_34, properties {soundWaveAssetPtr=/Game/Sounds/Fort_Weapons/SniperRifles/Sniper_Rifle_Fire_Close_New_03}}
 {id=WavePlayer_35, properties {soundWaveAssetPtr=/Game/Sounds/Fort_Weapons/SniperRifles/Sniper_Rifle_Fire_Close_New_02}}
 {id=WavePlayer_38, properties {soundWaveAssetPtr=/Game/Sounds/Fort_Weapons/SniperRifles/Sniper_Shot_DistantNew_A_01}}
 {id=WavePlayer_39, properties {soundWaveAssetPtr=/Game/Sounds/Fort_Weapons/SniperRifles/Sniper_Shot_DistantNew_A_02}}
 {id=WavePlayer_4, properties {soundWaveAssetPtr=/Game/Sounds/Fort_Weapons/SniperRifles/Sniper_Rifle_Fire_MediumNew_New_02}}
]
properties {
 maxDistance=27500
 soundClassObject=&Fort_WeaponRanged_Fire_3P_PSM
 volumeMultiplier=1
 attenuationOverrides {distanceAlgorithm=NaturalSound, falloffDistance=4000, attenuationShapeExtents {x=1000}}
 concurrencyOverrides {bLimitToOwner=true, maxCount=3, resolutionRule=StopOldest}
 concurrencySet [&Weapon_Gunfire_Local_Conc, &Weapon_Gunfire_Global_Conc]
}
