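# Sound cue for firing the tier 4 Triple Tap triple-shot rifle, with near and distant layers under separate sound classes.
# The third-person fire branch mixes a sweetener with a briefly delayed, modulated stereo shot through small-rifle attenuation.
# The distant branch crossfades medium and distant layers by distance, each modulated with its own projectile attenuation.
# A concurrency override prevents new instances once the limit is hit.
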SoundCue /Game/Sounds/Fort_Weapons/Rifle/Weapon_RFL_TripleShot_T04_Triple_Tap_Cue
root=Mixer_11
nodes [
 {id=Attenuation_2, children [Modulator_6], properties {attenuationSettings=/Game/Sounds/Attenuation/Fort_Weapon_Projectile_Medium_StW_Att}}
 {id=Attenuation_3, children [Modulator_7], properties {attenuationSettings=/Game/Sounds/Attenuation/Fort_Weapon_Projectile_Distant_StW_Att}}
 {id=Attenuation_68, children [Mixer_16], properties {attenuationSettings=/Game/Sounds/Attenuation/Fort_Weapon_Rifle_Sml_Att}}
 {id=Delay_5, children [Modulator_1], properties {delayMax=0.1, delayMin=0.08}}
 {
  id=DistanceCrossFade_1
  children [Attenuation_2, Attenuation_3]
  properties {crossFadeInput [
    {fadeInDistanceEnd=100, fadeInDistanceStart=0, fadeOutDistanceEnd=3000, fadeOutDistanceStart=2000, volume=1}
    {fadeInDistanceEnd=3000, fadeInDistanceStart=1000, fadeOutDistanceEnd=12000, fadeOutDistanceStart=10000, volume=1}
   ]}
 }
 {id=Mixer_11, children [SoundClass_47, SoundClass_48], properties {inputVolume [1,1]}}
 {id=Mixer_16, children [WavePlayer_177, Delay_5], properties {inputVolume [1,1]}}
 {id=Modulator_1, children [WavePlayer_1]}
 {id=Modulator_6, children [WavePlayer_26]}
 {id=Modulator_7, children [WavePlayer_25]}
 {
  id=SoundClass_47
  children [Attenuation_68]
  properties {soundClassOverride=&Fort_WeaponRanged_Fire_3P_PSM}
 }
 {
  id=SoundClass_48
  children [DistanceCrossFade_1]
  properties {soundClassOverride=&Fort_WeaponRanged_Fire_Distant}
 }
 {id=WavePlayer_1, properties {soundWaveAssetPtr=&/Rifle_Fire_Stereo_01}}
 {id=WavePlayer_177, properties {soundWaveAssetPtr=/Game/Sounds/Fort_Weapons/WeaponSweeteners/weapon_sweet_05}}
 {id=WavePlayer_25, properties {soundWaveAssetPtr=&/rifle_medium_02}}
 {id=WavePlayer_26, properties {soundWaveAssetPtr=&/rifle_medium_distant_02}}
]
properties {
 bOverrideConcurrency=true
 maxDistance=20000
 volumeMultiplier=0.8
 attenuationOverrides {distanceAlgorithm=NaturalSound, falloffDistance=4000, attenuationShapeExtents {x=1000}}
 concurrencyOverrides {resolutionRule=PreventNew}
}
